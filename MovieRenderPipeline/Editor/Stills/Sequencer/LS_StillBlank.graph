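# Level Sequence under the Movie Render Pipeline editor stills folder, apparently a blank template for rendering a still shot.
# A Cine Camera Actor with a spawn track holds a fixed transform, while its camera component sets constant focal length, manual focus distance and aperture.
# All these values are channel defaults with no animation, and a camera cut track spans one second at 30 fps.
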LevelSequence /MovieRenderPipeline/Editor/Stills/Sequencer/LS_StillBlank
bindings [
 CameraComponent {
  id=#1
  parent_id=#2
  possessed_class=CineCameraComponent
  tracks [
   {
    type=MovieSceneFloatTrack
    display_name=Current Focal Length
    sections [{type=MovieSceneFloatSection, channels [None {type=MovieSceneScriptingFloatChannel, default=35}]}]
   }
   {
    type=MovieSceneFloatTrack
    display_name="Manual Focus Distance (Focus Settings)"
    sections [{type=MovieSceneFloatSection, channels [None {type=MovieSceneScriptingFloatChannel, default=100000}]}]
   }
   {
    type=MovieSceneFloatTrack
    display_name=Current Aperture
    sections [{type=MovieSceneFloatSection, channels [None {type=MovieSceneScriptingFloatChannel, default=2.8}]}]
   }
  ]
 }
 Cine Camera Actor {
  id=#2
  possessed_class=CineCameraActor
  tracks [
   {
    type=MovieScene3DTransformTrack
    sections [{
      type=MovieScene3DTransformSection
      channels [
       Location.X {type=MovieSceneScriptingDoubleChannel, default=-945.0672}
       Location.Y {type=MovieSceneScriptingDoubleChannel, default=-820.1761}
       Location.Z {type=MovieSceneScriptingDoubleChannel, default=959.3351}
       Rotation.X {type=MovieSceneScriptingDoubleChannel, default=0.0001}
       Rotation.Y {type=MovieSceneScriptingDoubleChannel, default=-35.9995}
       Rotation.Z {type=MovieSceneScriptingDoubleChannel, default=13.2007}
       Scale.X {type=MovieSceneScriptingDoubleChannel, default=1}
       Scale.Y {type=MovieSceneScriptingDoubleChannel, default=1}
       Scale.Z {type=MovieSceneScriptingDoubleChannel, default=1}
      ]
     }]
   }
   {type=MovieSceneSpawnTrack, sections [{type=MovieSceneSpawnSection, channels [None {type=MovieSceneScriptingBoolChannel, default=true}]}]}
  ]
 }
]
display_rate {denominator=1, numerator=30}
tick_resolution {denominator=1, numerator=24000}
tracks [{type=MovieSceneCameraCutTrack, sections [{type=MovieSceneCameraCutSection, end_frame=30, start_frame=0}]}]
