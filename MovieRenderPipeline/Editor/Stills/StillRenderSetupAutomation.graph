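# Editor utility widget that creates still-render Level Sequences from selected or all level CineCameraActors and can queue them in Movie Render Queue.
# Both buttons call GenerateSequences, which builds each sequence through embedded Python (mrq_stills) and optionally allocates a job with the editor map and preset.
# ConfigureSettings then applies optional resolution, folder, tiling, sampling and filename overrides from the Details View.
# Blueprint sequence-building, DisableTracks and GenerateShotConfigFromSettings logic appear unused.
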EditorUtilityWidgetBlueprint /MovieRenderPipeline/Editor/Stills/StillRenderSetupAutomation
graphs [
 ConfigureSettings {nodes [
   ConfigureSettings() {is=title K2Node_FunctionEntry}
   > Set CustomizedPreset(CustomizedPreset:Movie Pipeline Primary Config Object Reference=GetConfiguration!(self:Movie Pipeline Executor Job Object Reference=Get PipelineExecutorJob() -> (PipelineExecutorJob:Movie Pipeline Executor Job Object Reference)) -> (ReturnValue:Movie Pipeline Primary Config Object Reference)) -> (Output_Get:Movie Pipeline Primary Config Object Reference)
   > n1 = FindOrAddSettingByClass!(
    execute
    self:Movie Pipeline Config Base Object Reference=Get CustomizedPreset() -> (CustomizedPreset:Movie Pipeline Primary Config Object Reference)
    InClass:Movie Pipeline Setting Class Reference=/Script/MovieRenderPipelineCore.MoviePipelineOutputSetting
    bIncludeDisabledSettings:bool=false
    bExactMatch:bool=false
   ) -> (then, ReturnValue:Movie Pipeline Output Setting Object Reference)
   > n2 = Branch(Condition=Get bOverrideMovieResolution() -> (bOverrideMovieResolution:bool))
   > n3 = Set OutputResolution!(
    execute
    OutputResolution:Int Point Structure=Get MovieResolution() -> (MovieResolution:Int Point Structure)
    self:Movie Pipeline Output Setting Object Reference=@n1
   ) -> (then, Output_Get:Int Point Structure)
   n4 = Branch(execute=[@n3.then, @n2.else], Condition=Get bOverrideOutputFolder() -> (bOverrideOutputFolder:bool))
   > n5 = Set OutputDirectory!(
    execute
    OutputDirectory:Directory Path Structure=Get OutputFolder() -> (OutputFolder:Directory Path Structure)
    self:Movie Pipeline Output Setting Object Reference=@n1
   ) -> (then, Output_Get:Directory Path Structure)
   n6 = FindOrAddSettingByClass(
    execute=[@n5.then, @n4.else]
    self=Get CustomizedPreset()
    InClass=/Script/MovieRenderPipelineCore.MoviePipelineHighResSetting
    bIncludeDisabledSettings=false
    bExactMatch=false
   ) -> (ReturnValue:Movie Pipeline High Res Setting Object Reference)
   > n7 = Branch(Condition=Get bOverrideTileCount() -> (bOverrideTileCount:bool))
   > n8 = Set TileCount!(execute, TileCount:int=Get TileCount() -> (TileCount:int=0), self:Movie Pipeline High Res Setting Object Reference=@n6) -> (then, Output_Get:int=0)
   n9 = FindOrAddSettingByClass(
    execute=[@n8.then, @n7.else]
    self=Get CustomizedPreset()
    InClass=/Script/MovieRenderPipelineCore.MoviePipelineAntiAliasingSetting
    bIncludeDisabledSettings=false
    bExactMatch=false
   ) -> (ReturnValue:Movie Pipeline Anti Aliasing Setting Object Reference)
   > n10 = Branch(Condition=Get bOverrideSpatialSampleCount() -> (bOverrideSpatialSampleCount:bool))
   > Set SpatialSampleCount!(
    execute
    SpatialSampleCount:int=Get SpatialSampleCount() -> (SpatialSampleCount:int=0)
    self:Movie Pipeline Anti Aliasing Setting Object Reference=@n9
   ) -> (then, Output_Get:int=0)
   > n11 = Set bOverrideAntiAliasing!(execute, bOverrideAntiAliasing:bool=true, self:Movie Pipeline Anti Aliasing Setting Object Reference=@n9) -> (then, Output_Get:bool)
   n12 = Branch(execute=[@n11.then, @n10.else], Condition=Get bOverrideTemporalSampleCount() -> (bOverrideTemporalSampleCount:bool))
   > n13 = Set TemporalSampleCount!(
    execute
    TemporalSampleCount:int=Get TemporalSampleCount() -> (TemporalSampleCount:int=0)
    self:Movie Pipeline Anti Aliasing Setting Object Reference=@n9
   ) -> (then, Output_Get:int=0)
   n14 = FindOrAddSettingByClass(
    execute=[@n13.then, @n12.else]
    self=Get CustomizedPreset()
    InClass=&MoviePipelineOutputSetting
    bIncludeDisabledSettings=false
    bExactMatch=false
   ) -> (ReturnValue:Movie Pipeline Output Setting Object Reference)
   > Branch(Condition=Get bUseMapAsOutputPrefix() -> (bUseMapAsOutputPrefix:bool))
   > GetEditorWorld(self=Get UnrealEditorSubsystem())
   > Set FileNameFormat!(execute, FileNameFormat:String="{level_name}_{sequence_name}_{frame_number}", self:Movie Pipeline Output Setting Object Reference=@n14) -> (then, Output_Get:String)
  ]}
 DisableTracks {nodes [
   DisableTracks() {is=title K2Node_FunctionEntry}
   n1 = Get MoviePipelineQueueSubsystem() -> (ReturnValue:Movie Pipeline Queue Subsystem Object Reference) {is=title K2Node_GetEditorSubsystem}
   > n2 = For Each Loop(Array:Array of Movie Pipeline Executor Job Object References=GetJobs!(self:Movie Pipeline Queue Object Reference=GetQueue!(self:Movie Pipeline Queue Subsystem Object Reference=@n1) -> (ReturnValue:Movie Pipeline Queue Object Reference)) -> (ReturnValue:Array of Movie Pipeline Executor Job Object References)) -> (Array Element:Movie Pipeline Executor Job Object Reference)
   n3 = Set ShotInfo!(execute, ShotInfo:Array of Movie Pipeline Executor Shot Object References, self:Movie Pipeline Executor Job Object Reference=@n2) -> (then, Output_Get:Array of Movie Pipeline Executor Shot Object References)
   > n4 = For Each Loop(Array:Array of Movie Pipeline Executor Shot Object References=@n3) -> (Array Element:Movie Pipeline Executor Shot Object Reference)
   GetShotOverrideConfiguration!(self:Movie Pipeline Executor Shot Object Reference=@n4) -> (ReturnValue:Movie Pipeline Shot Config Object Reference)
  ]}
 EventGraph {nodes [
   n1 = GetSelectedActors()
   > n2 = For Each Loop(Array:Array of Actor Object References=@n1) -> (Array Element:Actor Object Reference)
   > n3 = Construct Object from Class() {
    pins are=other
    also {
     inputs [
      execute {type=Exec}
      Class {type=Object Class Reference, default=/Script/LevelSequenceEditor.LevelSequenceFactoryNew}
      self {type=Object Reference}
      self {type=Self Object Reference, default=/MovieRenderPipeline/Editor/Stills/Sequencer/LS_StillBlank}
     ]
     outputs [
      then {type=Exec}
      ReturnValue {type=Level Sequence Factory New Object Reference}
     ]
    }
   }
   > n4 = GetAssetTools()
   n5 = GetDisplayName(Object=@n2)
   n6 = "To String (Integer)"(InInt=@n2.Array Index)
   > n7 = CreateAsset(
    self=@n4
    AssetName=Select(
     Option 0:String
     Option 1:String=Append!(A=LS_, B=@n5, C:String=_, D:String=@n6) -> (ReturnValue)
     Index=Get bUseUniqueAssetNames() -> (bUseUniqueAssetNames:bool)
    ) -> (ReturnValue:String)
    PackagePath=Get SequenceCreationPath() -> (SequenceCreationPath:String)
    AssetClass=/Script/LevelSequence
    Factory=@n3
   )
   > n8 = Cast To LevelSequence!(execute, Object=@n7) -> (then, CastFailed, AsLevel Sequence:Level Sequence Object Reference) {is=title K2Node_DynamicCast}
   > SetPlaybackEnd!(execute, Sequence:Movie Scene Sequence Object Reference=@n8, EndFrame:int=1) -> (then)
   > n9 = AddPossessable!(execute, Sequence:Movie Scene Sequence Object Reference=@n8, ObjectToPossess:Object Reference=@n2) -> (then, ReturnValue:Movie Scene Binding Proxy Structure)
   > n10 = Set Binding(Binding:Movie Scene Binding Proxy Structure=@n9) -> (Output_Get:Movie Scene Binding Proxy Structure)
   > n11 = AddTrack!(
    execute
    InBinding:"Movie Scene Binding Proxy Structure (by ref)"=@n10
    TrackType:Movie Scene Track Class Reference=/Script/MovieSceneTracks.MovieSceneCameraCutTrack
   ) -> (then, ReturnValue:Movie Scene Track Object Reference)
   > n12 = AddSection!(execute, Track:Movie Scene Track Object Reference=@n11) -> (then, ReturnValue:Movie Scene Section Object Reference)
   > SetStartFrame!(execute, Section:Movie Scene Section Object Reference=@n12, StartFrame:int=-1) -> (then)
   > SetEndFrame!(execute, Section:Movie Scene Section Object Reference=@n12, EndFrame:int=1) -> (then)
   > n13 = Cast To MovieSceneCameraCutSection!(execute, Object=@n12) -> (then, CastFailed, AsMovie Scene Camera Cut Section:Movie Scene Camera Cut Section Object Reference) {is=title K2Node_DynamicCast}
   > SetEditorProperty(
    Object=@n12
    PropertyName=CameraBindingID
    PropertyValue:"Movie Scene Object Binding ID Structure (by ref)"=GetCameraBindingID!(self:Movie Scene Camera Cut Section Object Reference=@n13) -> (ReturnValue:"Movie Scene Object Binding ID Structure (by ref)")
   )
   "On Clicked (CreateSequenceBtn)"() {is=title K2Node_ComponentBoundEvent}
   > n14 = GetSelectedActors()
   > n15 = For Each Loop(Array=@n14)
   > n16 = Cast To CineCameraActor!(execute, Object=@n15) -> (then, CastFailed, AsCine Camera Actor:Cine Camera Actor Object Reference) {is=title K2Node_DynamicCast}
   > Is Valid(InputObject=@n16)
   n17 = GetDisplayName(Object=@n15)
   n18 = "To String (Integer)"(InInt=RandomIntegerInRange!(Min:int=100, Max:int=800) -> (ReturnValue:int=0))
   > GenerateSequences!(
    execute
    self
    TargetCamera:Actor Object Reference=@n15
    JobAutoName:String=Select(
     Option 0=Append!(A=LS_, B=@n17) -> (ReturnValue)
     Option 1=Append!(A=LS_, B=@n17, C:String=_, D:String=@n18) -> (ReturnValue)
     Index=Get bUseUniqueAssetNames()
    )
    SequenceCreationPath:String=Get SequenceCreationPath()
   ) -> (then)
   "On Clicked (CreateFromAllCams)"() {is=title K2Node_ComponentBoundEvent}
   > n19 = GetAllActorsOfClass!(execute, WorldContextObject:Object Reference, ActorClass:Actor Class Reference=/Script/CinematicCamera.CineCameraActor) -> (then, OutActors:Array of Cine Camera Actor Object References)
   > n20 = For Each Loop(Array:Array of Cine Camera Actor Object References=@n19) -> (Array Element:Cine Camera Actor Object Reference)
   > Is Valid(InputObject=@n20)
   n21 = GetDisplayName(Object=@n20)
   n22 = "To String (Integer)"(InInt=RandomIntegerInRange(Min=100, Max=800) -> (ReturnValue=0))
   > GenerateSequences(
    TargetCamera=@n20
    JobAutoName=Select(
     Option 0=Append!(A=LS_, B=@n21) -> (ReturnValue)
     Option 1=Append!(A=LS_, B=@n21, C:String=_, D:String=@n22) -> (ReturnValue)
     Index=Get bUseUniqueAssetNames()
    )
    SequenceCreationPath=Get SequenceCreationPath()
   )
   DelayLoad() {is=title K2Node_CustomEvent}
   > Delay!(execute, WorldContextObject:Object Reference, Duration=0.2) -> (then)
   n23 = Cast To CineCameraActor(Object=@n2)
   > n24 = AddPossessable(
    Sequence=@n8
    ObjectToPossess=GetCineCameraComponent!(self:Cine Camera Actor Object Reference=@n23) -> (ReturnValue:Cine Camera Component Object Reference)
   )
   > SetParent!(
    execute
    InBinding:"Movie Scene Binding Proxy Structure (by ref)"=@n24
    InParentBinding:"Movie Scene Binding Proxy Structure (by ref)"=Get Binding() -> (Binding:Movie Scene Binding Proxy Structure)
   ) -> (then)
   > AddTrack(InBinding=@n24, TrackType=/Script/MovieSceneTracks.MovieSceneFloatTrack)
   Event PreConstruct()
   > SetObject(self=Get SequenceSettings() -> (SequenceSettings:Details View Object Reference), NewObject=Self-Reference())
   n25 = ProjectSavedDir!() -> (ReturnValue:String)
   > Set OutputFolder(OutputFolder:Directory Path Structure=Make Directory Path!(Path:String=Append!(A=@n25, B=VideoCaptures/) -> (ReturnValue)) -> (DirectoryPath:Directory Path Structure) {is=title K2Node_MakeStruct}) -> (Output_Get:Directory Path Structure)
   SetEditorProperty(Object=@n13, PropertyName=Guid, PropertyValue:"Wildcard (by ref)")
  ]}
 GenerateSequences {nodes [
   n1 = GenerateSequences!() -> (then, TargetCamera:Actor Object Reference, JobAutoName:String, SequenceCreationPath:String='/Game/Sequencer/Stills/') {is=title K2Node_FunctionEntry}
   > n2 = Set JobAutoName(JobAutoName:String=@n1.JobAutoName) -> (Output_Get:String)
   > n3 = Execute Python Script!(
    execute
    PythonScript="import mrq_stills\r\n\r\nsequence = mrq_stills.create_sequence_from_selection(asset_name=in_asset_name, length_frames = in_length_frames, package_path=in_package_path, target_camera=in_target_camera)"
    in_asset_name:String=@n2
    in_length_frames:int=Get SequenceLength() -> (SequenceLength:int=0)
    in_package_path:String=@n1.SequenceCreationPath
    in_target_camera:Actor Object Reference=@n1
   ) -> (then, ReturnValue, sequence:Level Sequence Object Reference) {inputs [in_asset_name, in_length_frames, in_package_path, in_target_camera], outputs [sequence]}
   > n4 = Sequence()
   > Branch(Condition=Get bAutomaticallyAddJobsToMovieQueue() -> (bAutomaticallyAddJobsToMovieQueue:bool))
   > n5 = Set SequenceOut(SequenceOut:Level Sequence Object Reference=@n3.sequence) -> (Output_Get:Level Sequence Object Reference)
   > PrintString(InString=GetDisplayName(Object=@n5), TextColor=(R=0.1428,G=0,B=1,A=1))
   > n6 = AllocateNewJob!(
    execute
    self:Movie Pipeline Queue Object Reference=GetQueue(self=Get MoviePipelineQueueSubsystem())
    InJobType:Movie Pipeline Executor Job Class Reference=/Script/MovieRenderPipelineCore.MoviePipelineExecutorJob
   ) -> (then, ReturnValue:Movie Pipeline Executor Job Object Reference)
   n7 = Get SequenceOut() -> (SequenceOut:Level Sequence Object Reference)
   > Set Sequence!(
    execute
    Sequence:Soft Object Path Structure=MakeSoftObjectPath!(PathString:String=Get Object Path String(Object=@n7)) -> (ReturnValue:Soft Object Path Structure)
    self:Movie Pipeline Executor Job Object Reference=@n6
   ) -> (then, Output_Get:Soft Object Path Structure)
   > Set JobName!(execute, JobName:String=Get JobAutoName() -> (JobAutoName:String), self:Movie Pipeline Executor Job Object Reference=@n6) -> (then, Output_Get:String)
   > SetConfiguration!(
    execute
    self:Movie Pipeline Executor Job Object Reference=@n6
    InPreset:Movie Pipeline Primary Config Object Reference=Get JobSettingsPreset() -> (JobSettingsPreset:Movie Pipeline Primary Config Object Reference)
   ) -> (then)
   > n8 = GetEditorWorld(self=Get UnrealEditorSubsystem())
   > Set Map!(
    execute
    Map:Soft Object Path Structure=MakeSoftObjectPath(PathString=Get Object Path String(Object=@n8))
    self:Movie Pipeline Executor Job Object Reference=@n6
   ) -> (then, Output_Get:Soft Object Path Structure)
   > Set Author!(execute, Author:String=Auto, self:Movie Pipeline Executor Job Object Reference=@n6) -> (then, Output_Get:String)
   > Set PipelineExecutorJob(PipelineExecutorJob:Movie Pipeline Executor Job Object Reference=@n6) -> (Output_Get:Movie Pipeline Executor Job Object Reference)
   DelayLoad(execute=@n4.then_1) {is=title K2Node_CallFunction}
   > ConfigureSettings() {is=title K2Node_CallFunction}
  ]}
 GenerateShotConfigFromSettings {nodes [
   GenerateShotConfigFromSettings() {is=title K2Node_FunctionEntry}
   > n1 = Construct Object from Class(Class=/Script/MovieRenderPipelineCore.MoviePipelinePrimaryConfig) -> (ReturnValue:Movie Pipeline Primary Config Object Reference)
   > Set OutMasterConfig(OutMasterConfig:Movie Pipeline Primary Config Object Reference=@n1) -> (Output_Get:Movie Pipeline Primary Config Object Reference)
   > n2 = Sequence()
   > n3 = Sequence!(execute) -> (then_0, then_1, then_2:Exec, then_3:Exec, then_4:Exec)
   > FindOrAddSettingByClass(
    self=Get OutMasterConfig() -> (OutMasterConfig:Movie Pipeline Primary Config Object Reference)
    InClass=/Script/MovieRenderPipelineRenderPasses.MoviePipelineDeferredPassBase
    bIncludeDisabledSettings=false
    bExactMatch=false
   ) -> (ReturnValue:Movie Pipeline Deferred Pass Base Object Reference)
   > FindOrAddSettingByClass(
    self=Get OutMasterConfig()
    InClass=/Script/MovieRenderPipelineRenderPasses.MoviePipelineImageSequenceOutput_JPG
    bIncludeDisabledSettings=false
    bExactMatch=false
   ) -> (ReturnValue:Movie Pipeline Image Sequence Output JPG Object Reference)
   n4 = FindOrAddSettingByClass(
    execute=@n3.then_1
    self=Get OutMasterConfig()
    InClass=&MoviePipelineOutputSetting
    bIncludeDisabledSettings=false
    bExactMatch=false
   ) -> (ReturnValue:Movie Pipeline Output Setting Object Reference)
   > Set OutputResolution(OutputResolution=Get MovieResolution(), self=@n4)
   > Set OutputDirectory(OutputDirectory=Get OutputFolder(), self=@n4)
   n5 = FindOrAddSettingByClass(
    execute=@n3.then_2
    self=Get OutMasterConfig()
    InClass=&MoviePipelineHighResSetting
    bIncludeDisabledSettings=false
    bExactMatch=false
   ) -> (ReturnValue:Movie Pipeline High Res Setting Object Reference)
   > Set TileCount(TileCount=Get TileCount() -> (TileCount=0), self=@n5) -> (Output_Get=0)
   n6 = FindOrAddSettingByClass(
    execute=@n3.then_3
    self=Get OutMasterConfig()
    InClass=&MoviePipelineAntiAliasingSetting
    bIncludeDisabledSettings=false
    bExactMatch=false
   ) -> (ReturnValue:Movie Pipeline Anti Aliasing Setting Object Reference)
   > Set SpatialSampleCount(SpatialSampleCount=Get SpatialSampleCount() -> (SpatialSampleCount=0), self=@n6) -> (Output_Get=0)
   > Set TemporalSampleCount(TemporalSampleCount=Get TemporalSampleCount() -> (TemporalSampleCount=0), self=@n6) -> (Output_Get=0)
   Return Node!(execute=@n2.then_1, ReturnValue:Movie Pipeline Primary Config Object Reference=Get OutMasterConfig())
  ]}
 IsRenderEnabled {nodes [
   IsRenderEnabled() {is=title K2Node_FunctionEntry}
   > Return Node(ReturnValue:bool=IsValid(Object=Get TargetSequence() -> (TargetSequence:Level Sequence Object Reference)))
  ]}
]
variables [
 bAutomaticallyAddJobsToMovieQueue {type=boolean, default=true}
 Binding {type=MovieSceneBindingProxy, default {bindingId=#1, sequence=null}}
 bOverrideMovieResolution {type=boolean}
 bOverrideOutputFolder {type=boolean}
 bOverrideSpatialSampleCount {type=boolean}
 bOverrideTemporalSampleCount {type=boolean}
 bOverrideTileCount {type=boolean}
 bUseMapAsOutputPrefix {type=boolean}
 bUseUniqueAssetNames {type=boolean, default=true}
 ConstrainAspectRatio {type=boolean, default=true}
 CustomizedPreset {type=MoviePipelinePrimaryConfig, default=null}
 Filmback {
  type=CameraFilmbackSettings
  default {sensorAspectRatio=1.33, sensorHeight=13.365, sensorHorizontalOffset=0, sensorVerticalOffset=0, sensorWidth=23.76}
 }
 JobAutoName {type=string, default=""}
 JobSettingsPreset {type=MoviePipelinePrimaryConfig, default=/MovieRenderPipeline/Editor/Stills/Presets/Still_Draft}
 LensSettings {
  type=LensSettings
  default {
   chromaticAberration=0
   bloom {
    intensity=0.675
    method=BM_SOG
    convolution {
     bufferScale=0.133
     intensity=1
     preFilterMax=15000
     preFilterMin=7
     preFilterMult=15
     scatterDispersion=1
     size=1
     texture=null
     centerUV=<0.5 0.5>
    }
    gaussianSum {
     filter1Size=0.3
     filter2Size=1
     filter3Size=2
     filter4Size=10
     filter5Size=30
     filter6Size=64
     intensity=1
     sizeScale=4
     threshold=-1
     filter1Tint=<rgba 0.3465 0.3465 0.3465 1>
     filter2Tint=<rgba 0.138 0.138 0.138 1>
     filter3Tint=<rgba 0.1176 0.1176 0.1176 1>
     filter4Tint=<rgba 0.066 0.066 0.066 1>
     filter5Tint=<rgba 0.066 0.066 0.066 1>
     filter6Tint=<rgba 0.061 0.061 0.061 1>
    }
   }
   imperfections {dirtMask=null, dirtMaskIntensity=0, dirtMaskTint=<rgba 0.5 0.5 0.5 1>}
  }
 }
 MovieResolution {type=IntPoint, default=<1920 1080>}
 NextRenderExecutor {type=Class, default=null}
 OutputFolder {type=DirectoryPath, default {path=""}}
 Override Camera Settings {type=boolean}
 PipelineExecutorJob {type=MoviePipelineExecutorJob, default=null}
 RenderQueueWithExecutor {type=MoviePipelineExecutorBase, default=null}
 SelectedActors {type="array<CineCameraActor>", default []}
 SequenceCreationPath {type=string, default=&Stills/}
 SequenceLength {type=integer, default=1}
 SequenceOut {type=LevelSequence, default=null}
 SpatialSampleCount {type=integer, default=1}
 TargetCamera {type=string, default=""}
 TargetPackagePath {type=string, default=/Game/Sequencer/MovieRenderPipeline/Configs/}
 TargetSequence {type=LevelSequence, default=null}
 TemporalSampleCount {type=integer, default=1}
 TileCount {type=integer, default=1}
]
widgets {
 info {rootWidgetClass=&VerticalBox, widgetCount=14}
 widgets [
  VerticalBox_19 {widgetClassPath=&VerticalBox, widget=VerticalBox_19}
  VerticalBox {widgetClassPath=&VerticalBox, parent=VerticalBox_19, widget=VerticalBox, slot {class=VerticalBoxSlot}}
  TextBlock_1 {
   widgetClassPath=&TextBlock
   parent=VerticalBox
   widget=TextBlock_1
   properties {text=Sequencer Still Setup}
   slot {class=VerticalBoxSlot, properties {horizontalAlignment=HAlign_Center}}
  }
  HorizontalBox {widgetClassPath=&HorizontalBox, parent=VerticalBox, widget=HorizontalBox, slot {class=VerticalBoxSlot}}
  CreateSequenceBtn {
   widgetClassPath=&Button
   parent=HorizontalBox
   widget=CreateSequenceBtn
   properties {
    backgroundColor=<rgba 0 0.225 1 1>
    widgetStyle {
     clickedSlateSound {resourceObject=null}
     disabled {
      bIsDynamicallyLoaded=false
      drawAs=RoundedBox
      imageType=NoImage
      mirroring=NoMirror
      resourceName=None
      resourceObject=null
      tiling=NoTile
      imageSize=<0 0>
      margin=<ltrb 0 0 0 0>
      outlineSettings {
       bUseBrushTransparency=false
       roundingType=FixedRadius
       width=1
       color {colorUseRule=UseColor_ColorTable, specifiedColor=<rgba 0 0 0 0>}
       cornerRadii=<4 4 4 4>
      }
      tintColor {colorUseRule=UseColor_Specified, specifiedColor=<rgba 0.0395 0.0395 0.0395 1>}
      uVRegion {bIsValid=false, max=<0 0>, min=<0 0>}
     }
     disabledForeground {colorUseRule=UseColor_ColorTable, specifiedColor=<rgba 0 0 0 0>}
     hovered {
      bIsDynamicallyLoaded=false
      drawAs=RoundedBox
      imageType=NoImage
      mirroring=NoMirror
      resourceName=None
      resourceObject=null
      tiling=NoTile
      imageSize=<0 0>
      margin=<ltrb 0 0 0 0>
      outlineSettings {
       bUseBrushTransparency=false
       roundingType=FixedRadius
       width=1
       color {colorUseRule=UseColor_ColorTable, specifiedColor=<rgba 0 0 0 0>}
       cornerRadii=<4 4 4 4>
      }
      tintColor {colorUseRule=UseColor_Specified, specifiedColor=<rgba 0.0953 0.0953 0.0953 1>}
      uVRegion {bIsValid=false, max=<0 0>, min=<0 0>}
     }
     hoveredForeground {colorUseRule=UseColor_ColorTable, specifiedColor=<rgba 0 0 0 0>}
     hoveredSlateSound {resourceObject=null}
     normal {
      bIsDynamicallyLoaded=false
      drawAs=RoundedBox
      imageType=NoImage
      mirroring=NoMirror
      resourceName=None
      resourceObject=null
      tiling=NoTile
      imageSize=<0 0>
      margin=<ltrb 0 0 0 0>
      outlineSettings {
       bUseBrushTransparency=false
       roundingType=FixedRadius
       width=1
       color {colorUseRule=UseColor_ColorTable, specifiedColor=<rgba 0 0 0 0>}
       cornerRadii=<4 4 4 4>
      }
      tintColor {colorUseRule=UseColor_Specified, specifiedColor=<rgba 0.0395 0.0395 0.0395 1>}
      uVRegion {bIsValid=false, max=<0 0>, min=<0 0>}
     }
     normalForeground {colorUseRule=UseColor_ColorTable, specifiedColor=<rgba 0 0 0 0>}
     normalPadding=<ltrb 12 1.5 12 1.5>
     pressed {
      bIsDynamicallyLoaded=false
      drawAs=RoundedBox
      imageType=NoImage
      mirroring=NoMirror
      resourceName=None
      resourceObject=null
      tiling=NoTile
      imageSize=<0 0>
      margin=<ltrb 0 0 0 0>
      outlineSettings {
       bUseBrushTransparency=false
       roundingType=FixedRadius
       width=1
       color {colorUseRule=UseColor_ColorTable, specifiedColor=<rgba 0 0 0 0>}
       cornerRadii=<4 4 4 4>
      }
      tintColor {colorUseRule=UseColor_Specified, specifiedColor=<rgba 0.0284 0.0284 0.0284 1>}
      uVRegion {bIsValid=false, max=<0 0>, min=<0 0>}
     }
     pressedForeground {colorUseRule=UseColor_ColorTable, specifiedColor=<rgba 0 0 0 0>}
     pressedPadding=<ltrb 12 2.5 12 0.5>
     pressedSlateSound {resourceObject=null}
    }
   }
   slot {class=HorizontalBoxSlot, properties {horizontalAlignment=HAlign_Left, padding=<ltrb 5 5 5 5>}}
  }
  TextBlock_3 {
   widgetClassPath=&TextBlock
   parent=CreateSequenceBtn
   widget=TextBlock_3
   properties {
    text=Create Sequence For Selected
    font {
     bForceMonospaced=false
     bMaterialIsStencil=false
     fontMaterial=null
     fontObject=&Roboto
     letterSpacing=0
     monospacedWidth=1
     size=12
     skewAmount=0
     typefaceFontName=Bold
     outlineSettings {
      bApplyOutlineToDropShadows=false
      bMiteredCorners=false
      bSeparateFillAlpha=false
      outlineMaterial=null
      outlineSize=0
      outlineColor=<rgba 0 0 0 1>
     }
    }
   }
   slot {class=ButtonSlot}
  }
  CreateFromAllCams {
   widgetClassPath=&Button
   parent=HorizontalBox
   widget=CreateFromAllCams
   properties {
    backgroundColor=<rgba 0.3639 0.0478 1 1>
    widgetStyle {
     clickedSlateSound {resourceObject=null}
     disabled {
      bIsDynamicallyLoaded=false
      drawAs=RoundedBox
      imageType=NoImage
      mirroring=NoMirror
      resourceName=None
      resourceObject=null
      tiling=NoTile
      imageSize=<0 0>
      margin=<ltrb 0 0 0 0>
      outlineSettings {
       bUseBrushTransparency=false
       roundingType=FixedRadius
       width=1
       color {colorUseRule=UseColor_ColorTable, specifiedColor=<rgba 0 0 0 0>}
       cornerRadii=<4 4 4 4>
      }
      tintColor {colorUseRule=UseColor_Specified, specifiedColor=<rgba 0.0395 0.0395 0.0395 1>}
      uVRegion {bIsValid=false, max=<0 0>, min=<0 0>}
     }
     disabledForeground {colorUseRule=UseColor_ColorTable, specifiedColor=<rgba 0 0 0 0>}
     hovered {
      bIsDynamicallyLoaded=false
      drawAs=RoundedBox
      imageType=NoImage
      mirroring=NoMirror
      resourceName=None
      resourceObject=null
      tiling=NoTile
      imageSize=<0 0>
      margin=<ltrb 0 0 0 0>
      outlineSettings {
       bUseBrushTransparency=false
       roundingType=FixedRadius
       width=1
       color {colorUseRule=UseColor_ColorTable, specifiedColor=<rgba 0 0 0 0>}
       cornerRadii=<4 4 4 4>
      }
      tintColor {colorUseRule=UseColor_Specified, specifiedColor=<rgba 0.0953 0.0953 0.0953 1>}
      uVRegion {bIsValid=false, max=<0 0>, min=<0 0>}
     }
     hoveredForeground {colorUseRule=UseColor_ColorTable, specifiedColor=<rgba 0 0 0 0>}
     hoveredSlateSound {resourceObject=null}
     normal {
      bIsDynamicallyLoaded=false
      drawAs=RoundedBox
      imageType=NoImage
      mirroring=NoMirror
      resourceName=None
      resourceObject=null
      tiling=NoTile
      imageSize=<0 0>
      margin=<ltrb 0 0 0 0>
      outlineSettings {
       bUseBrushTransparency=false
       roundingType=FixedRadius
       width=1
       color {colorUseRule=UseColor_ColorTable, specifiedColor=<rgba 0 0 0 0>}
       cornerRadii=<4 4 4 4>
      }
      tintColor {colorUseRule=UseColor_Specified, specifiedColor=<rgba 0.0395 0.0395 0.0395 1>}
      uVRegion {bIsValid=false, max=<0 0>, min=<0 0>}
     }
     normalForeground {colorUseRule=UseColor_ColorTable, specifiedColor=<rgba 0 0 0 0>}
     normalPadding=<ltrb 12 1.5 12 1.5>
     pressed {
      bIsDynamicallyLoaded=false
      drawAs=RoundedBox
      imageType=NoImage
      mirroring=NoMirror
      resourceName=None
      resourceObject=null
      tiling=NoTile
      imageSize=<0 0>
      margin=<ltrb 0 0 0 0>
      outlineSettings {
       bUseBrushTransparency=false
       roundingType=FixedRadius
       width=1
       color {colorUseRule=UseColor_ColorTable, specifiedColor=<rgba 0 0 0 0>}
       cornerRadii=<4 4 4 4>
      }
      tintColor {colorUseRule=UseColor_Specified, specifiedColor=<rgba 0.0284 0.0284 0.0284 1>}
      uVRegion {bIsValid=false, max=<0 0>, min=<0 0>}
     }
     pressedForeground {colorUseRule=UseColor_ColorTable, specifiedColor=<rgba 0 0 0 0>}
     pressedPadding=<ltrb 12 2.5 12 0.5>
     pressedSlateSound {resourceObject=null}
    }
   }
   slot {class=HorizontalBoxSlot, properties {horizontalAlignment=HAlign_Left, padding=<ltrb 5 5 5 5>}}
  }
  TextBlock_4 {
   widgetClassPath=&TextBlock
   parent=CreateFromAllCams
   widget=TextBlock_4
   properties {
    text=Create Sequence For All Cameras
    font {
     bForceMonospaced=false
     bMaterialIsStencil=false
     fontMaterial=null
     fontObject=&Roboto
     letterSpacing=0
     monospacedWidth=1
     size=12
     skewAmount=0
     typefaceFontName=Bold
     outlineSettings {
      bApplyOutlineToDropShadows=false
      bMiteredCorners=false
      bSeparateFillAlpha=false
      outlineMaterial=null
      outlineSize=0
      outlineColor=<rgba 0 0 0 1>
     }
    }
   }
   slot {class=ButtonSlot}
  }
  Spacer {
   widgetClassPath=&Spacer
   parent=HorizontalBox
   widget=Spacer
   slot {class=HorizontalBoxSlot, properties {size {sizeRule=Fill, value=1}}}
  }
  SequenceSettings {
   widgetClassPath=&DetailsView
   parent=VerticalBox
   widget=SequenceSettings
   properties {categoriesToShow [Still Settings]}
   slot {class=VerticalBoxSlot, properties {padding=<ltrb 3 3 3 3>}}
  }
  VerticalBox_95 {widgetClassPath=&VerticalBox, parent=VerticalBox_19, widget=VerticalBox_95, slot {class=VerticalBoxSlot}}
  HorizontalBox_1 {widgetClassPath=&HorizontalBox, parent=VerticalBox_19, widget=HorizontalBox_1, slot {class=VerticalBoxSlot}}
  Spacer_93 {
   widgetClassPath=&Spacer
   parent=HorizontalBox_1
   widget=Spacer_93
   slot {class=HorizontalBoxSlot, properties {size {sizeRule=Fill, value=1}}}
  }
  Spacer_841 {widgetClassPath=&Spacer, parent=VerticalBox_19, widget=Spacer_841, slot {class=VerticalBoxSlot}}
 ]
}
